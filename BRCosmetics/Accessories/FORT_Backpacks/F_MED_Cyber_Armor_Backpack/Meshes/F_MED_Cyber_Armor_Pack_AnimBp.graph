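# Animation Blueprint for the Cyber Armor backpack cosmetic; its wired output only scales the pack bone to 0.9 on the input pose.
# Unconnected leftovers include an Anim Dynamics spring chain for four plates with a skydiving blend, and an additive bone-translation chain.
# The update event still sets a time-based customSpeed and zeroes Target_off, the dynamics alpha, while targeting.
# The random-vector wobble calls feeding the translation variables are not hooked to that event.
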AnimBlueprint /BRCosmetics/Accessories/FORT_Backpacks/F_MED_Cyber_Armor_Backpack/Meshes/F_MED_Cyber_Armor_Pack_AnimBp
target_skeleton=&/F_MED_Cyber_Armor_Pack_Skeleton
graphs [
 AnimGraph {nodes [
   n1 = Get Target_off() -> (Target_off:double)
   n2 = Anim Dynamics - Bone: dyn_plate_l_top(ComponentPose=Local To Component(LocalPose=Use cached pose 'Pose'()), Alpha=@n1, ExternalForce="0,0,0") {
    angularDampingOverride=0.9
    angularSpringConstant=200000
    bGravityOverrideInSimSpace=true
    bLinearSpring=true
    bOverrideAngularDamping=true
    bOverrideLinearDamping=true
    gravityScale=0
    linearDampingOverride=0.995
    linearSpringConstant=75
    simulationSpace=BoneRelative
    alphaBoolBlend {blendOption=Linear}
    boundBone {boneName=dyn_plate_l_top}
    componentAppliedLinearAccClamp=<12.5 12.5 12.5>
    componentLinearAccScale=<0.05 0.1 0.2>
    componentLinearVelScale=<9 9 9>
    gravityOverride {z=-980}
    physicsBodyDefinitions [{
      sphereCollisionRadius=0
      boundBone {boneName=dyn_plate_l_top}
      boxExtents=<125 125 125>
      constraintSetup {angularLimitsMax=<0 0 0>, angularLimitsMin=<0 0 0>, angularTarget=<0 0 0>, linearAxesMax=<0.5 1.5 1.5>, linearAxesMin=<-1.5 -0.5 -0.5>}
      localJointOffset=<-0.2 0 2>
     }]
    relativeSpaceBone {boneName=pack}
   }
   n3 = Anim Dynamics - Bone: dyn_plate_r_top(ComponentPose=@n2, Alpha=@n1, ExternalForce="0,0,0") {
    angularDampingOverride=0.9
    angularSpringConstant=200000
    bGravityOverrideInSimSpace=true
    bLinearSpring=true
    bOverrideAngularDamping=true
    bOverrideLinearDamping=true
    gravityScale=0
    linearDampingOverride=0.995
    linearSpringConstant=75
    simulationSpace=BoneRelative
    alphaBoolBlend {blendOption=Linear}
    boundBone {boneName=dyn_plate_r_top}
    componentAppliedLinearAccClamp=<12.5 12.5 12.5>
    componentLinearAccScale=<0.05 0.1 0.2>
    componentLinearVelScale=<9 9 9>
    gravityOverride {z=-980}
    physicsBodyDefinitions [{
      sphereCollisionRadius=0
      boundBone {boneName=dyn_plate_r_top}
      boxExtents=<125 125 125>
      constraintSetup {angularLimitsMax=<0 0 0>, angularLimitsMin=<0 0 0>, angularTarget=<0 0 0>, linearAxesMax=<0.5 2 0.5>, linearAxesMin=<-2 -0.5 -2>}
      localJointOffset=<0.5 0 0.2>
     }]
    relativeSpaceBone {boneName=pack}
   }
   n4 = Anim Dynamics - Bone: dyn_plate_l_bot(ComponentPose=@n3, Alpha=@n1, ExternalForce="0,0,0") {
    angularDampingOverride=0.9
    angularSpringConstant=200000
    bGravityOverrideInSimSpace=true
    bLinearSpring=true
    bOverrideAngularDamping=true
    bOverrideLinearDamping=true
    gravityScale=0
    linearDampingOverride=0.995
    linearSpringConstant=75
    simulationSpace=BoneRelative
    alphaBoolBlend {blendOption=Linear}
    boundBone {boneName=dyn_plate_l_bot}
    componentAppliedLinearAccClamp=<12.5 12.5 12.5>
    componentLinearAccScale=<0.05 0.1 0.2>
    componentLinearVelScale=<9 9 9>
    gravityOverride {z=-980}
    physicsBodyDefinitions [{
      sphereCollisionRadius=0
      boundBone {boneName=dyn_plate_l_bot}
      boxExtents=<125 125 125>
      constraintSetup {angularLimitsMax=<0 0 0>, angularLimitsMin=<0 0 0>, angularTarget=<0 0 0>, linearAxesMax=<2 2 2>, linearAxesMin=<-0.5 -0.5 -0.5>}
      localJointOffset=<-0.5 0 1>
     }]
    relativeSpaceBone {boneName=pack}
   }
   n5 = Anim Dynamics - Bone: dyn_plate_r_bot(ComponentPose=@n4, Alpha=@n1, ExternalForce="0,0,0") {
    angularDampingOverride=0.9
    angularSpringConstant=200000
    bGravityOverrideInSimSpace=true
    bLinearSpring=true
    bOverrideAngularDamping=true
    bOverrideLinearDamping=true
    gravityScale=0
    linearDampingOverride=0.995
    linearSpringConstant=75
    simulationSpace=BoneRelative
    alphaBoolBlend {blendOption=Linear}
    boundBone {boneName=dyn_plate_r_bot}
    componentAppliedLinearAccClamp=<12.5 12.5 12.5>
    componentLinearAccScale=<0.05 0.1 0.2>
    componentLinearVelScale=<9 9 9>
    gravityOverride {z=-980}
    physicsBodyDefinitions [{
      sphereCollisionRadius=0
      boundBone {boneName=dyn_plate_r_bot}
      boxExtents=<125 125 125>
      constraintSetup {angularLimitsMax=<0 0 0>, angularLimitsMin=<0 0 0>, angularTarget=<0 0 0>, linearAxesMax=<1 1 0.5>, linearAxesMin=<-0.5 -0.5 -1>}
      localJointOffset=<-0.3 0 0.6>
     }]
    relativeSpaceBone {boneName=pack}
   }
   Blend Poses by bool(BlendPose_0=Use cached pose 'Pose'(), BlendPose_1=Component To Local(ComponentPose=@n5), bActiveValue=Get bIsSkyDiving()) {blendType=Linear}
   Input Pose()
   n6 = "Transform (Modify) Bone - Bone: dyn_plate_l_top"!(ComponentPose, Alpha=1, Translation=Get NewVar_0() -> (NewVar_0:Vector)) -> (Pose) {translationMode=BMM_Additive, translationSpace=BCS_BoneSpace, alphaBoolBlend {blendOption=Linear}, boneToModify {boneName=dyn_plate_l_top}}
   n7 = "Transform (Modify) Bone - Bone: dyn_plate_r_top"!(ComponentPose=@n6, Alpha=1, Translation=Get NewVar_1() -> (NewVar_1:Vector)) -> (Pose) {translationMode=BMM_Additive, translationSpace=BCS_BoneSpace, alphaBoolBlend {blendOption=Linear}, boneToModify {boneName=dyn_plate_r_top}}
   n8 = "Transform (Modify) Bone - Bone: dyn_plate_l_bot"!(ComponentPose=@n7, Alpha=1, Translation=Get NewVar_2() -> (NewVar_2:Vector)) -> (Pose) {translationMode=BMM_Additive, translationSpace=BCS_BoneSpace, alphaBoolBlend {blendOption=Linear}, boneToModify {boneName=dyn_plate_l_bot}}
   "Transform (Modify) Bone - Bone: dyn_plate_r_bot"!(ComponentPose=@n8, Alpha=1, Translation=Get NewVar_3() -> (NewVar_3:Vector)) -> (Pose) {translationMode=BMM_Additive, translationSpace=BCS_BoneSpace, alphaBoolBlend {blendOption=Linear}, boneToModify {boneName=dyn_plate_r_bot}}
   n9 = "Transform (Modify) Bone - Bone: pack"(ComponentPose=Local To Component(LocalPose=Input Pose() {name=InPose_0}), Translation="0,0,0", Rotation="0,0,0", Scale="0.9,0.9,0.9") {scaleMode=BMM_Replace, alphaBoolBlend {blendOption=Linear}, boneToModify {boneName=pack}, scale=<0.9 0.9 0.9>}
   Output Pose(Result=Component To Local(ComponentPose=@n9)) {layerGroup=None}
  ]}
 EventGraph {nodes [
   n1 = Random Vector Function!(
    execute
    CharacterPartAnimInstance:Custom Character Part Anim Instance Object Reference
    SkelMeshComponent:Mesh Component Object Reference
    JointName:Name=None
    GameTimeInSeconds:double=0
    TimeOffset:double=Get customSpeed() -> (customSpeed:double)
    PerlinScale:double=0
    Frequency:Vector="-0.25,-0.35,-0.25"
    Amplitude:Vector="0.5,0.15,0.5"
    Offset:Vector="0,0,0"
   ) -> (then, VectorOut:Vector)
   > Set NewVar_2(NewVar_2:Vector=@n1) -> (Output_Get:Vector)
   n2 = Random Vector Function(
    JointName=None
    GameTimeInSeconds=0
    TimeOffset=Get customSpeed()
    PerlinScale=0
    Frequency="0.7,0.7,0.3"
    Amplitude="0.4,0.2,0.2"
    Offset="0,0,0"
   )
   > Set NewVar_0(NewVar_0:Vector=@n2) -> (Output_Get:Vector)
   n3 = Random Vector Function(
    JointName=None
    GameTimeInSeconds=0
    TimeOffset=Get customSpeed()
    PerlinScale=0
    Frequency="-0.8,-0.6,-0.8"
    Amplitude="0.2,0.4,0.2"
    Offset="0,0,0"
   )
   > Set NewVar_1(NewVar_1:Vector=@n3) -> (Output_Get:Vector)
   n4 = Random Vector Function(
    JointName=None
    GameTimeInSeconds=0
    TimeOffset=Get customSpeed()
    PerlinScale=0
    Frequency="0.6,0.5,0.35"
    Amplitude="0.2,0.3,0.45"
    Offset="0,0,0"
   )
   > Set NewVar_3(NewVar_3:Vector=@n4) -> (Output_Get:Vector)
   Event BlueprintUpdateAnimation()
   > n5 = Sequence!(execute) -> (then_0, then_1, then_2:Exec, then_3:Exec, then_4:Exec, then_5:Exec)
   Set customSpeed(execute=@n5.then_4, customSpeed:double=float / float(A=GetRealTimeSeconds(), B=1.25) {is=title K2Node_CallFunction}) -> (Output_Get:double)
   n6 = Branch(execute=@n5.then_5, Condition=Get bIsTargeting())
   > Set Target_off(Target_off:double=0) -> (Output_Get:double)
   Set Target_off(execute=@n6.else, Target_off=1)
   TryGetPawnOwner()
  ]}
]
variables [
 customSpeed {type=number}
 Floating_Rotator1 {type=Rotator}
 NewVar_0 {type=Vector}
 NewVar_1 {type=Vector}
 NewVar_2 {type=Vector}
 NewVar_3 {type=Vector}
 Target_off {type=number}
]
